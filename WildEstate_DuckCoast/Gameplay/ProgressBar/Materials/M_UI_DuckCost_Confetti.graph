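# Translucent UI material for a Duck Coast progress bar that, despite its Confetti name, draws bubbles and a wave.
# Particle grids make randomly sized, randomly culled circle outlines, with UVs panned at per-column random speeds.
# Panning Perlin noise distorts a rotated box into the wave.
# A mask blends both layers with opacity parameters; Progress cutoff, rounded 9-slice frame and lerped purple color are also defined.
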Material /WildEstate_DuckCoast/Gameplay/ProgressBar/Materials/M_UI_DuckCost_Confetti
graph {
 properties {bAutomaticallySetUsageInEditor=false, blendMode=BLEND_Translucent, materialDomain=MD_UI, usedShadingModels=Default Lit}
 nodes [
  n1 = Multiply(A=NamedRerouteUsage(), B=OneMinus(None=NamedRerouteUsage()))
  Multiply(
   A=MaterialFunctionCall!(A=NamedRerouteUsage(), B=ScalarParameter() {defaultValue=0.35, parameterName=FillChance}) {materialFunction=/FortUILibrary/MaterialFunctions/Basics/MF_UI_BooleanOperations}.IsGreater
   B=@n1
  )
  n2 = NamedRerouteUsage()
  n3 = OneMinus(None=NamedRerouteUsage())
  n4 = Add(
   A=Multiply(A=ScalarParameter() {defaultValue=1, parameterName=Bubbles Opacity}, B=@n1)
   B=Multiply(A=Multiply(A=@n2, B=@n3), B=ScalarParameter() {defaultValue=1, parameterName=Wave Opacity})
  )
  n5 = NamedRerouteUsage()
  NamedRerouteDeclaration(None=Multiply(A=Multiply(A=@n4, B=@n5), B=ScalarParameter() {defaultValue=1, parameterName=Master Opacity})) {name=Mask, nodeColor=<rgba 0 0 0 1>}
  n6 = NamedRerouteUsage()
  n7 = Sine(None=DotProduct(A=Floor(None=@n6), B=Constant2Vector() {g=45.5555, r=33.5556}))
  n8 = Constant() {r=0.25}
  NamedRerouteDeclaration(None=Frac(None=Multiply(A=@n7, B=@n8))) {name=Random, nodeColor=<rgba 0 0.2031 1 1>}
  n9 = NamedRerouteUsage()
  n10 = Sine(None=DotProduct(A=Floor(None=@n9), B=Constant2Vector() {r=0.2}))
  n11 = Constant() {r=0.4}
  NamedRerouteDeclaration(None=Frac(None=Multiply(A=@n10, B=@n11))) {name=Random2, nodeColor=<rgba 0 0.2031 1 1>}
  n12 = Constant() {r=2}
  n13 = MaterialFunctionCall(Float2=MaterialFunctionCall() {materialFunction=&GetUserInterfaceUV}.Pixel Size) {materialFunction=&BreakOutFloat2Components}
  n14 = Multiply(A=@n12, B=Divide(A=@n13.R, B=@n13.R))
  NamedRerouteDeclaration(None=@n14) {name=Bubble Amount + AR, nodeColor=<rgba 0 0 0 1>}
  NamedRerouteDeclaration(None=AppendVector(A=@n14, B=@n12)) {name=Bubble Amount, nodeColor=<rgba 0 0 0 1>}
  n15 = Time()
  n16 = TextureCoordinate()
  n17 = NamedRerouteUsage()
  n18 = Floor(None=Multiply(A=ComponentMask(None=@n16) {r=true}, B=@n17))
  n19 = LinearInterpolate(Alpha=MaterialFunctionCall!(UV=Divide(A=@n18, B=@n17), Seed=Constant() {r=2}) {materialFunction=&MF_UI_Random}.Result) {constA=1, constB=3}
  NamedRerouteDeclaration(None=Panner(Time=Multiply(A=@n15, B=@n19)) {speedY=0.4}) {name=Bubble UV, nodeColor=<rgba 0 0 0 1>}
  NamedRerouteDeclaration(None=LinearInterpolate(
    A=Constant3Vector() {constant=<rgba 0.3184 0.0625 0.7767 1>}
    B=VectorParameter() {parameterName=Col1, defaultValue=<rgba 0.5015 0.239 0.9715 1>}.RGB
    Alpha=NamedRerouteUsage()
   )) {name=Color, nodeColor=<rgba 0 0 0 1>}
  n20 = Constant()
  n21 = MaterialFunctionCall!(
   UV=NamedRerouteUsage()
   Global Offset
   Global Rotation
   Tiling=NamedRerouteUsage()
   Cell Offset=@n20
   Cell Scale=Constant() {r=1}
   Cell Rotation=@n20
   Random Seed=Constant() {r=2}
  ) {materialFunction=/FortUILibrary/MaterialFunctions/Basics/MF_UI_ParticleGrid}
  NamedRerouteDeclaration(None=@n21.Random Cell Values) {name=Random Bubble, nodeColor=<rgba 0 0 0 1>}
  n22 = Constant()
  n23 = MaterialFunctionCall(
   UV=NamedRerouteUsage()
   Tiling=NamedRerouteUsage()
   Cell Offset=@n22
   Cell Scale=Constant() {r=1}
   Cell Rotation=@n22
   Random Seed=NamedRerouteUsage()
  ) {materialFunction=&MF_UI_ParticleGrid}
  n24 = LinearInterpolate(
   A=ScalarParameter() {defaultValue=0.4, parameterName=Bubble Size Min}
   B=ScalarParameter() {defaultValue=0.8, parameterName=Bubble Size Max}
   Alpha=@n23.Random Cell Values
  ) {constA=0.3, constB=0.6}
  n25 = Constant() {r=0.15}
  n26 = Constant() {r=-1}
  n27 = ScalarParameter() {defaultValue=0.2, parameterName=Bubble Random Mask}
  NamedRerouteDeclaration(None=Multiply(
    A=MaterialFunctionCall(UVs=@n23.Cells UV, Size=@n24, OutlineThickness=@n25, OutlineAlignment=@n26) {materialFunction=&MF_UI_Circle}.Outline
    B=Step(Y=@n21.Random Cell Values, X=@n27) {constX=0.6}
   ) {constA=1}) {name=Bubble, nodeColor=<rgba 0 0 0 1>}
  NamedRerouteDeclaration(None=Divide(A=ScalarParameter() {defaultValue=640, parameterName=Width}, B=ScalarParameter() {defaultValue=32, parameterName=Height})) {name=AspectRatio, nodeColor=<rgba 1 0 0.6094 1>}
  n28 = TextureCoordinate()
  n29 = ComponentMask(None=@n28) {r=true}
  n30 = NamedRerouteUsage()
  NamedRerouteDeclaration(None=AppendVector(A=Multiply(A=@n29, B=@n30), B=ComponentMask(None=@n28) {g=true})) {name=UV_Fixed, nodeColor=<rgba 0 0.9297 1 1>}
  n31 = MaterialFunctionCall!(UVs) {materialFunction=/FortUILibrary/MaterialFunctions/Basics/MF_UI_9SliceUVs}
  n32 = Constant2Vector() {g=0.85, r=0.85}
  n33 = Constant() {r=0.15}
  NamedRerouteDeclaration(None=OneMinus(None=MaterialFunctionCall(UVs=@n31.UVs, Size=@n32, CornerRoundness=@n33) {materialFunction=&MF_UI_Box}.Fill)) {name=BossHealthBar_Progress, nodeColor=<rgba 0 0 0 1>}
  n34 = Subtract(
   A=ComponentMask(None=MaterialFunctionCall() {materialFunction=&GetUserInterfaceUV}.Normalized UV) {r=true}
   B=ScalarParameter() {defaultValue=1, group=Progress, parameterName=Progress, sliderMax=1}
  )
  NamedRerouteDeclaration(None=OneMinus(None=Step(X=@n34) {constX=0})) {name=ProgressGradient, nodeColor=<rgba 0 0 0 1>}
  n35 = Frac(None=NamedRerouteUsage())
  n36 = ComponentMask(None=@n35) {r=true}
  n37 = Divide(A=NamedRerouteUsage()) {constB=7}
  n38 = Add(A=ComponentMask(None=@n35) {g=true}, B=@n37)
  n39 = Add(A=@n37) {constB=0.1}
  NamedRerouteDeclaration(None=MaterialFunctionCall(UVs=AppendVector(A=@n36, B=@n38), Size=Saturate(None=@n39)) {materialFunction=&MF_UI_Circle}.Outline) {name=Fill, nodeColor=<rgba 0 0 0 1>}
  n40 = Add(A=TextureCoordinate(), B=Constant2Vector() {g=-0.5})
  n41 = TextureSampleParameter2D(UVs=Panner(Coordinate=TextureCoordinate()) {speedY=0.4}) {parameterName=NoiseTexture, texture=/WildEstate_DuckCoast/Gameplay/ProgressBar/Textures/T_UI_RTL_PerlinNoise2}
  n42 = ScalarParameter() {defaultValue=0.5, parameterName=Intensity}
  n43 = AppendVector(A=ComponentMask(None=@n40) {r=true}, B=Add(A=ComponentMask(None=@n40) {g=true}, B=Multiply(A=@n41.R, B=@n42) {constB=0.056}))
  n44 = Constant() {r=0.25}
  NamedRerouteDeclaration(None=MaterialFunctionCall(
    UVs=MaterialFunctionCall(UVs=@n43, "Rotation Angle (0-1)"=@n44) {materialFunction=&CustomRotator}.Rotated Values
    Size=Constant2Vector() {g=1, r=0.5}
   ) {materialFunction=&MF_UI_Box}.Fill) {name=Wave, nodeColor=<rgba 0.2422 0 1 1>}
  NamedRerouteUsage()
  NamedRerouteUsage()
  NamedRerouteUsage()
 ]
}
