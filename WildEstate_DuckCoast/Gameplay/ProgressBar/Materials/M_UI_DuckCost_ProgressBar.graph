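# Translucent UI material for a Duck Coast progress bar, with parameters for fill, empty and mark colors plus master opacity.
# A Progress scalar thresholds horizontal UV into a hard fill mask and, via two offset Steps, a thin toggleable edge marker.
# Procedural rounded boxes on 9-slice UVs supply backplate, progress-area and outline masks, using BossHealthBar naming for the first two.
# Named reroutes carry masks into layered color lerps and opacity math.
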Material /WildEstate_DuckCoast/Gameplay/ProgressBar/Materials/M_UI_DuckCost_ProgressBar
graph {
 properties {bAutomaticallySetUsageInEditor=false, blendMode=BLEND_Translucent, materialDomain=MD_UI, usedShadingModels=Default Lit}
 nodes [
  n1 = MaterialFunctionCall() {materialFunction=&GetUserInterfaceUV}
  ComponentMask(None=@n1.Normalized UV) {g=true, r=true}
  MaterialFunctionCall() {materialFunction=&GetUserInterfaceUV}
  n2 = LinearInterpolate(
   A=LinearInterpolate(
    A=VectorParameter() {parameterName=EmptyColor, defaultValue=<rgba 0 0 0 1>}.RGB
    B=VectorParameter() {parameterName=FillColor, defaultValue=<rgba 0.1912 0.0222 0.6038 1>}.RGB
    Alpha=NamedRerouteUsage()
   )
   B=VectorParameter() {parameterName=EmptyColor, defaultValue=<rgba 0 0 0 1>}.RGB
   Alpha=NamedRerouteUsage()
  )
  n3 = Multiply(A=Multiply(A=NamedRerouteUsage()) {constB=0}, B=NamedRerouteUsage())
  NamedRerouteDeclaration(None=LinearInterpolate(
    A=LinearInterpolate(A=@n2, Alpha=@n3)
    B=VectorParameter() {parameterName=Progress Mark Color, defaultValue=<rgba 1 1 1 1>}.RGB
    Alpha=NamedRerouteUsage()
   )) {name=Final Color, nodeColor=<rgba 0 0 0 1>}
  n4 = NamedRerouteUsage()
  n5 = NamedRerouteUsage()
  n6 = NamedRerouteUsage()
  n7 = Multiply(A=@n5, B=LinearInterpolate(A=Multiply(A=@n5) {constB=0.4}, B=@n6, Alpha=@n6) {constAlpha=0})
  NamedRerouteDeclaration(None=Multiply(A=Add(A=@n4, B=@n7), B=ScalarParameter() {defaultValue=1, parameterName=Master Opacity})) {name=opacity, nodeColor=<rgba 0 0.4609 1 1>}
  NamedRerouteDeclaration(None=MaterialFunctionCall(
    UVs=MaterialFunctionCall!(UVs) {materialFunction=/FortUILibrary/MaterialFunctions/Basics/MF_UI_9SliceUVs}.UVs
    Size=Constant2Vector() {g=1, r=1}
    CornerRoundness=Constant() {r=0.3}
   ) {materialFunction=&MF_UI_Box}.Fill) {name=BossHealthBar_Backplate, nodeColor=<rgba 0 0 0 1>}
  n8 = MaterialFunctionCall() {materialFunction=&MF_UI_9SliceUVs}
  n9 = Constant2Vector() {g=0.85, r=0.85}
  n10 = Constant() {r=0.15}
  NamedRerouteDeclaration(None=OneMinus(None=MaterialFunctionCall(UVs=@n8.UVs, Size=@n9, CornerRoundness=@n10) {materialFunction=&MF_UI_Box}.Fill)) {name=BossHealthBar_Progress, nodeColor=<rgba 0 0 0 1>}
  n11 = Subtract(
   A=ComponentMask(None=MaterialFunctionCall() {materialFunction=&GetUserInterfaceUV}.Normalized UV) {r=true}
   B=ScalarParameter() {defaultValue=1, group=Progress, parameterName=Progress, sliderMax=1}
  )
  NamedRerouteDeclaration(None=OneMinus(None=Step(X=@n11) {constX=0})) {name=ProgressGradient, nodeColor=<rgba 0 0 0 1>}
  n12 = ScalarParameter() {defaultValue=1, group=Progress, parameterName=Progress, sliderMax=1}
  n13 = ComponentMask(None=@n1.Normalized UV) {r=true}
  n14 = Subtract(A=Step(Y=@n12, X=@n13), B=Step(Y=Add(A=@n12) {constB=0.005}, X=@n13))
  n15 = ScalarParameter() {defaultValue=1, parameterName=Show Progress Mark}
  n16 = NamedRerouteUsage()
  NamedRerouteDeclaration(None=Multiply(A=LinearInterpolate(B=@n14, Alpha=@n15), B=OneMinus(None=@n16))) {name=Progress Mark, nodeColor=<rgba 0 0 0 1>}
  NamedRerouteDeclaration(None=MaterialFunctionCall(
    UVs=MaterialFunctionCall() {materialFunction=&MF_UI_9SliceUVs}.UVs
    Size=Constant2Vector() {g=0.7, r=0.7}
    CornerRoundness=Constant() {r=0.02}
    OutlineThickness=Constant() {r=0.06}
   ) {materialFunction=&MF_UI_Box}.Outline) {name=LightOutline, nodeColor=<rgba 0 0 0 1>}
  NamedRerouteUsage()
  NamedRerouteUsage()
 ]
}
